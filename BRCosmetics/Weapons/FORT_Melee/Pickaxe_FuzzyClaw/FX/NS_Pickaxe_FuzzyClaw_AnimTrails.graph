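# Niagara swing-trail effect for the FuzzyClaw pickaxe, combining animation trails with scattered feather meshes.
# Three ribbon emitters (soft dark, grey, thin bright tip) each build three anim trails from skeletal mesh bones or sockets.
# Trail colours fade by curve, with time-of-day tints and a shared system alpha scale.
# Feathers spawn by distance moved, gated by system age, launch along a socket's rotation, then drift with drag and growing curl noise.
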NiagaraSystem /BRCosmetics/Weapons/FORT_Melee/Pickaxe_FuzzyClaw/FX/NS_Pickaxe_FuzzyClaw_AnimTrails
dependencies {
 usedDataInterfaces [&NiagaraDataInterfaceColorCurve, &NiagaraDataInterfaceCurve, &NiagaraDataInterfaceSkeletalMesh]
 usedDynamicInputs []
 usedModules [
  &self:SystemSpawnScript.SystemScriptSource.SystemScriptGraph.NiagaraNodeAssignment_0.SetVariables_#1
  &self:SystemSpawnScript.SystemScriptSource.SystemScriptGraph.NiagaraNodeAssignment_2.SetVariables_#2
  &SystemState
  &self:SystemSpawnScript.SystemScriptSource.SystemScriptGraph.NiagaraNodeAssignment_3.SetVariables_#3
 ]
}
emitters {
 MainGreyTrail {
  emitterUpdateScript [
   &EmitterState()
   &FN_SpawnAnimTrail(Number of Anim Trails=3)
  ]
  particleSpawnScript [
   &FN_InitializeAnimTrailBoneIndecies()
   &FN_InitializeAnimTrailBoneIndecies(AnimTrail Selection=1, Second Bone or Socket Index=2) {module=FN_InitializeAnimTrailBoneIndecies001}
   &FN_InitializeAnimTrailBoneIndecies(AnimTrail Selection=2, Second Bone or Socket Index=3) {module=FN_InitializeAnimTrailBoneIndecies002}
   &FN_InitializeAnimTrail(
    AnimTrailWidth=1.1
    Color=&FN_Select_Color_from_Time_Of_Day(Morning=<rgba 1 1 1 1>, Day=<rgba 1 1 1 1>, Evening=<rgba 1 1 1 1>, Night=<rgba 1 1 1 1>)
    Lifetime=0.1
    SkeletalMesh=$System.SkeletalMesh
   )
   &AddVelocity(Velocity=$User.DebugSpeed)
  ]
  particleUpdateScript [
   &FN_UpdateAnimTrail(SkeletalMesh=$System.SkeletalMesh) {module=FN_UpdateAnimTrail001}
   &Color(Color=&ColorFromCurve(ColorCurve=DataInterface() {propertyValues {
       AlphaCurve {keys [<at 0 0.8>,<at 0.1816 0 RCIM_Cubic>]}
       BlueCurve {keys [<at 0 0.1>]}
       GreenCurve {keys [<at 0 0.1>]}
       RedCurve {keys [<at 0 0.1>]}
      }}))
   /SetVariables(Particles.Color=&MultiplyLinearColor(
     Linear Color=&MultiplyLinearColor(
      Linear Color=$Particles.Color
      Scale Factor=&MakeVector4_FromLinearColor(LinearColor=&FNTimeOfDayColor(Night Color=<rgba 0.3958 0.5714 1 1>, Morning Color=<rgba 1 1 1 1>, Day Color=<rgba 1 1 1 1>, Evening Color=<rgba 1 1 1 1>))
     )
     Scale Factor=&MakeVector4(X=1, Y=1, Z=1, W=$System.ScaleAlpha)
    )) {isSetParameters=true}
   &SolveForcesAndVelocity(Clamp Velocity=false, Limit Acceleration=false)
  ]
  rendererClasses [&NiagaraRibbonRendererProperties]
  renderers [{rendererClass=&NiagaraRibbonRendererProperties}]
 }
 NE_Feathers {
  emitterUpdateScript [
   &EmitterState()
   &SpawnPerUnit(
    Max Movement Threshold=4000
    Module.Use Max Movement Threshold=true
    Module.Use Movement Tolerance=false
    Spawn Probability:NiagaraFloat=&MakeCustomFloatFromBool(Bool=&SetBoolByFloatComparison(A=$Engine.System.Age, B=0.05), True Float=1, False Float=0)
    Spawn Spacing=40
    Use Spawn Probability=true
   )
  ]
  particleSpawnScript [
   &InitializeParticle(
    Color=<rgba 0 0.5718 1 1>
    Lifetime=&RandomRangeFloat(Minimum=0.35, Maximum=0.45)
    Mesh Scale Mode=::NewEnumerator3
    Mesh Uniform Scale=&RandomRangeFloat(Minimum=0.7, Maximum=1)
    Sprite Rotation Mode=::NewEnumerator2
    Sprite Size Mode=::NewEnumerator3
    UsePositionOffset=true
    Write Material Random=true
   )
   &AddVelocity(Velocity=$User.DebugSpeed, Velocity Speed Scale=0.2)
   &SkeletalMeshLocation(
    Bone / Socket Index=1
    MaskSpawnGroup=false
    Mesh Sampling Type=::NewEnumerator5
    Sampled Position Offset=&RandomRangeVector(Minimum=<0 0 -20>, Maximum=<0 0 5>)
    Skeletal Mesh=$System.SkeletalMesh
    Socket Sampling Mode=::NewEnumerator5
   )
   &AddVelocity(
    Rotation Coordinate Space=::Simulation
    Velocity=&RotateVectorByQuaternion(Vector To Rotate=<0 0 1>, Quaternion=$Particles.SkeletalMeshLocation.SampledRotation)
    Velocity Speed Scale=&RandomRangeFloat(Minimum=200, Maximum=250)
   ) {module=AddVelocity001}
   &InitialMeshOrientation(Use Rotation Vector=false)
  ]
  particleUpdateScript [
   &ParticleState()
   &ScaleMeshSize(Scale Factor=&VectorFromFloat(Value=&FloatFromCurve(FloatCurve=DataInterface() {propertyValues {Curve {keys [
          {arriveTangent=8.2127, interpMode=RCIM_Cubic, leaveTangent=8.2127, tangentMode=RCTM_User, time=0, value=0}
          {
           arriveTangent=0.0908
           arriveTangentWeight=0.232
           interpMode=RCIM_Cubic
           leaveTangent=0.0908
           leaveTangentWeight=2.2538
           tangentMode=RCTM_User
           time=0.1
           value=1
          }
          {arriveTangent=-0.042, interpMode=RCIM_Cubic, leaveTangent=-0.042, tangentMode=RCTM_User, time=1, value=0}
         ]}}})))
   &SkeletalMeshLocation(
    Bone / Socket Index=1
    MaskSpawnGroup=false
    Mesh Sampling Type=::NewEnumerator5
    Position Sampling=::NewEnumerator2
    Skeletal Mesh=$System.SkeletalMesh
    Socket Sampling Mode=::NewEnumerator5
   ) {module=SkeletalMeshLocation001}
   &ScaleSpriteSize(Uniform Curve Sprite Scale=DataInterface() {propertyValues {Curve {keys [
        {arriveTangent=8.2127, interpMode=RCIM_Cubic, leaveTangent=8.2127, tangentMode=RCTM_User, time=0, value=0}
        {
         arriveTangent=0.0908
         arriveTangentWeight=0.232
         interpMode=RCIM_Cubic
         leaveTangent=0.0908
         leaveTangentWeight=2.2538
         tangentMode=RCTM_User
         time=0.1
         value=1
        }
        {arriveTangent=-0.042, interpMode=RCIM_Cubic, leaveTangent=-0.042, tangentMode=RCTM_User, time=1, value=0}
       ]}}}) {module=ScaleSpriteSize001}
   &ScaleColor(ScaleA=true, ScaleRGB=true)
   &DynamicMaterialParameters(
    Index 0 Param 3=1
    Index 0 Param 4=1
    Module.Param0WriteEnabled=true
    Module.Param1WriteEnabled=true
    Module.Param2WriteEnabled=true
    Module.Param3WriteEnabled=true
   )
   &Drag(Drag=&RandomRangeFloat(Minimum=6, Maximum=8), Use Linear Drag=true, Use Rotational Drag=true)
   &CurlNoiseForce(
    Noise Strength=&FloatFromCurve(
     FloatCurve=DataInterface() {propertyValues {Curve {keys [
         {arriveTangent=2.7279, interpMode=RCIM_Cubic, leaveTangent=2.7279, tangentMode=RCTM_User, time=0, value=0}
         {interpMode=RCIM_Cubic, tangentMode=RCTM_User, time=1, value=1}
        ]}}}
     Scale Curve=550
    )
    Pan Noise=false
   )
   &SolveForcesAndVelocity(Clamp Velocity=false, Limit Acceleration=false)
   &UpdateMeshOrientation(Rotation Vector=&VectorFromFloat(Value=&RandomRangeFloat(Minimum=-1, Maximum=1)))
  ]
  rendererClasses [&NiagaraMeshRendererProperties]
  renderers [{rendererClass=&NiagaraMeshRendererProperties, rendererIndex=0}]
 }
 SoftDarkTrail {
  emitterUpdateScript [
   &EmitterState()
   &FN_SpawnAnimTrail(Number of Anim Trails=3)
  ]
  particleSpawnScript [
   &FN_InitializeAnimTrailBoneIndecies()
   &FN_InitializeAnimTrailBoneIndecies(AnimTrail Selection=1, Second Bone or Socket Index=2) {module=FN_InitializeAnimTrailBoneIndecies001}
   &FN_InitializeAnimTrailBoneIndecies(AnimTrail Selection=2, Second Bone or Socket Index=3) {module=FN_InitializeAnimTrailBoneIndecies002}
   &FN_InitializeAnimTrail(
    AnimTrailWidth=1.2
    Color=&FN_Select_Color_from_Time_Of_Day(Morning=<rgba 0 2.8891 4 1>, Day=<rgba 0 3.8853 5 1>, Evening=<rgba 0 3.5105 4 1>, Night=<rgba 0 3.8978 5 1>)
    Lifetime=0.1
    SkeletalMesh=$System.SkeletalMesh
   )
   &AddVelocity(Velocity=$User.DebugSpeed)
  ]
  particleUpdateScript [
   &FN_UpdateAnimTrail(SkeletalMesh=$System.SkeletalMesh) {module=FN_UpdateAnimTrail001}
   &Color(Color=&MergeHelper_ColorSwitch(
     0=&ColorFromCurve(ColorCurve=DataInterface() {propertyValues {
        AlphaCurve {keys [<at 0 0.5 RCIM_Cubic>,<at 1 0 RCIM_Cubic>]}
        BlueCurve {keys [<at 0 0.02>]}
        GreenCurve {keys [<at 0 0.0049>]}
        RedCurve {keys [<at 0 0>]}
       }})
     1=&ColorFromCurve(ColorCurve=DataInterface() {propertyValues {
        AlphaCurve {keys [<at 0 0.5 RCIM_Cubic>,<at 0.5899 0>]}
        BlueCurve {keys [<at 0 0.02>]}
        GreenCurve {keys [<at 0 0.0049>]}
        RedCurve {keys [<at 0 0>]}
       }})
     2=&ColorFromCurve(ColorCurve=DataInterface() {propertyValues {
        AlphaCurve {keys [<at 0 0.5 RCIM_Cubic>,<at 0.5952 0 RCIM_Cubic>]}
        BlueCurve {keys [<at 0 0.02>]}
        GreenCurve {keys [<at 0 0.0049>]}
        RedCurve {keys [<at 0 0>]}
       }})
     3=?
     Index=&SpawnGroup()
     Num=::NewEnumerator1
    ))
   /SetVariables(Particles.Color=&MultiplyLinearColor(
     Linear Color=&MultiplyLinearColor(
      Linear Color=$Particles.Color
      Scale Factor=&MakeVector4_FromLinearColor(LinearColor=&FNTimeOfDayColor(Night Color=<rgba 0.5 0.5 0.5 1>, Morning Color=<rgba 1 1 1 1>, Day Color=<rgba 1 1 1 1>, Evening Color=<rgba 1 1 1 1>))
     )
     Scale Factor=&MakeVector4(X=1, Y=1, Z=1, W=$System.ScaleAlpha)
    )) {isSetParameters=true}
   &SolveForcesAndVelocity(Clamp Velocity=false, Limit Acceleration=false)
  ]
  rendererClasses [&NiagaraRibbonRendererProperties]
  renderers [{rendererClass=&NiagaraRibbonRendererProperties}]
 }
 TipWhiteTrail {
  emitterUpdateScript [
   &EmitterState()
   &FN_SpawnAnimTrail(AnimTrailWidthMode=::FromSecond, Number of Anim Trails=3)
  ]
  particleSpawnScript [
   &FN_InitializeAnimTrailBoneIndecies()
   &FN_InitializeAnimTrailBoneIndecies(AnimTrail Selection=1, Second Bone or Socket Index=2) {module=FN_InitializeAnimTrailBoneIndecies001}
   &FN_InitializeAnimTrailBoneIndecies(AnimTrail Selection=2, Second Bone or Socket Index=3) {module=FN_InitializeAnimTrailBoneIndecies002}
   &FN_InitializeAnimTrail(AnimTrailWidth=0.075, Lifetime=0.22, SkeletalMesh=$System.SkeletalMesh)
   &AddVelocity(Velocity=$User.DebugSpeed)
  ]
  particleUpdateScript [
   &FN_UpdateAnimTrail(SkeletalMesh=$System.SkeletalMesh) {module=FN_UpdateAnimTrail001}
   &Color(Color=&MultiplyLinearColor(
     Linear Color=&MergeHelper_ColorSwitch(
      0=&ColorFromCurve(ColorCurve=DataInterface() {propertyValues {
         AlphaCurve {keys [<at 0 1>,<at 0.1816 0 RCIM_Cubic>]}
         BlueCurve {keys [<at 0 2>,<at 0.0588 0.6354>]}
         GreenCurve {keys [<at 0 2>,<at 0.0588 0.4779>]}
         RedCurve {keys [<at 0 2>,<at 0.0588 0.3956>]}
        }})
      1=&ColorFromCurve(ColorCurve=DataInterface() {propertyValues {
         AlphaCurve {keys [<at 0 0.5>,<at 0.0885 0 RCIM_Cubic>]}
         BlueCurve {keys [<at 0 2>,<at 0.0588 0.6354>]}
         GreenCurve {keys [<at 0 2>,<at 0.0588 0.5271>]}
         RedCurve {keys [<at 0 2>,<at 0.0588 0.4832>]}
        }})
      2=&ColorFromCurve(ColorCurve=DataInterface() {propertyValues {
         AlphaCurve {keys [<at 0 0.5>,<at 0.0903 0 RCIM_Cubic>]}
         BlueCurve {keys [<at 0 2>,<at 0.0588 0.6354>]}
         GreenCurve {keys [<at 0 2>,<at 0.0588 0.4882>]}
         RedCurve {keys [<at 0 2>,<at 0.0588 0.46>]}
        }})
      3=?
      Index=&SpawnGroup()
      Num=::NewEnumerator1
     )
     Scale Factor=&MakeVector4_FromLinearColor(LinearColor=&FNTimeOfDayColor(Night Color=<rgba 0.2448 0.4093 1 1>, Morning Color=<rgba 1 0.5437 0.3698 1>, Day Color=<rgba 1 1 1 1>, Evening Color=<rgba 1 0.6687 0 1>))
    ))
   /SetVariables(Particles.Color=&MultiplyLinearColor(
     Linear Color=&MultiplyLinearColor(
      Linear Color=$Particles.Color
      Scale Factor=&MakeVector4_FromLinearColor(LinearColor=&FNTimeOfDayColor(Night Color=<rgba 0.7344 0.7344 0.7344 1>, Morning Color=<rgba 1 1 1 1>, Day Color=<rgba 1 1 1 1>, Evening Color=<rgba 1 1 1 1>))
     )
     Scale Factor=&MakeVector4(X=1, Y=1, Z=1, W=$System.ScaleAlpha)
    )) {isSetParameters=true}
   &SolveForcesAndVelocity(Clamp Velocity=false, Limit Acceleration=false)
  ]
  rendererClasses [&NiagaraRibbonRendererProperties]
  renderers [{rendererClass=&NiagaraRibbonRendererProperties}]
 }
}
summary {
 systemName=NS_Pickaxe_FuzzyClaw_AnimTrails
 emitters [
  {emitterName=SoftDarkTrail, rendererClasses [&NiagaraRibbonRendererProperties]}
  {emitterName=MainGreyTrail, rendererClasses [&NiagaraRibbonRendererProperties]}
  {emitterName=TipWhiteTrail, rendererClasses [&NiagaraRibbonRendererProperties]}
  {emitterName=NE_Feathers, rendererClasses [&NiagaraMeshRendererProperties]}
 ]
 userVariables [
  User.DebugSpeed {type=Vector3f, description="", defaultValue {struct=Vector3f, value=<0 0 0>}}
  User.RibbonIDOffset {type=NiagaraInt32}
 ]
}
