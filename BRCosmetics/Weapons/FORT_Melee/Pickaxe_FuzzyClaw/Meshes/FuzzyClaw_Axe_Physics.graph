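# Physics asset for the FuzzyClaw pickaxe mesh, giving its claw and feather bones capsule bodies for secondary motion.
# A small base body anchors three claws and eight two-link feather chains.
# Each claw locks twist and one swing axis, leaving a narrow hinge-like swing.
# Feather constraints set zero twist limits and asymmetric swing limits, narrow on one axis and mostly wider on the other, varying per chain.
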PhysicsAsset /BRCosmetics/Weapons/FORT_Melee/Pickaxe_FuzzyClaw/Meshes/FuzzyClaw_Axe_Physics
bodies {
 dyn_base_joint [{shapeType=Capsule, length=0, radius=2.5, shapeName=dyn_base_joint_capsule}]
 dyn_claw_1_joint [{
   shapeType=Capsule
   length=19.5
   radius=1.5
   shapeName=dyn_claw_1_joint_capsule
   center=<12.755 -0.0003 1.1159>
   rotation=<rot 85 179.9962 179.9976>
  }]
 dyn_claw_2_joint [{
   shapeType=Capsule
   length=18
   radius=2
   shapeName=dyn_claw_2_joint_capsule
   center=<12.6426 -0.0003 1.1061>
   rotation=<rot 85 179.9961 179.9975>
  }]
 dyn_claw_3_joint [{
   shapeType=Capsule
   length=17.5
   radius=2
   shapeName=dyn_claw_3_joint_capsule
   center=<12.1465 -0.0002 1.0627>
   rotation=<rot 85 179.9965 179.9976>
  }]
 dyn_feather_10_joint [{
   shapeType=Capsule
   length=4.5
   radius=1
   shapeName=dyn_feather_10_joint_capsule
   center=<4.2094 0 -0.3683>
   rotation=<rot -85 179.9991 -179.9995>
  }]
 dyn_feather_11_joint [{
   shapeType=Capsule
   length=2
   radius=1
   shapeName=dyn_feather_11_joint_capsule
   center=<3.0825 0 -0.2697>
   rotation=<rot -85.0001 179.9993 -179.9996>
  }]
 dyn_feather_12_joint [{
   shapeType=Capsule
   length=2
   radius=1
   shapeName=dyn_feather_12_joint_capsule
   center=<3.0825 0 -0.2697>
   rotation=<rot -85.0001 179.9993 -179.9996>
  }]
 dyn_feather_13_joint [{
   shapeType=Capsule
   length=2
   radius=1
   shapeName=dyn_feather_13_joint_capsule
   center=<3.405 0 -0.2537>
   rotation=<rot -87.1585 0.0009 0.2614>
  }]
 dyn_feather_14_joint [{
   shapeType=Capsule
   length=2
   radius=1
   shapeName=dyn_feather_14_joint_capsule
   center=<3.0825 0 -0.2697>
   rotation=<rot -88.5558 179.9987 -179.9992>
  }]
 dyn_feather_15_joint [{shapeType=Capsule, length=4, radius=1, shapeName=dyn_feather_15_joint_capsule, center=<5.3113 0 -0.4647>, rotation=<rot -85 179.9996 -180>}]
 dyn_feather_16_joint [{
   shapeType=Capsule
   length=4.5
   radius=1.5
   shapeName=dyn_feather_16_joint_capsule
   center=<4.184 0 -0.3661>
   rotation=<rot -85 179.9993 -179.9997>
  }]
 dyn_feather_1_joint [{
   shapeType=Capsule
   length=2.5
   radius=1
   shapeName=dyn_feather_1_joint_capsule
   center=<3.5454 0 -0.2584>
   rotation=<rot -88.609 0.0015 -0.0019>
  }]
 dyn_feather_2_joint [{
   shapeType=Capsule
   length=2
   radius=1
   shapeName=dyn_feather_2_joint_capsule
   center=<3.3902 0 0.0713>
   rotation=<rot -86.5888 0.0004 8.0767>
  }]
 dyn_feather_3_joint [{
   shapeType=Capsule
   length=2
   radius=1
   shapeName=dyn_feather_3_joint_capsule
   center=<2.8473 0 0.0315>
   rotation=<rot -84.9999 -179.9999 179.9997>
  }]
 dyn_feather_4_joint [{
   shapeType=Capsule
   length=2.5
   radius=1
   shapeName=dyn_feather_4_joint_capsule
   center=<2.8949 0 -0.2533>
   rotation=<rot -84.9998 -179.9993 -175.3281>
  }]
 dyn_feather_5_joint [{
   shapeType=Capsule
   length=4
   radius=1
   shapeName=dyn_feather_5_joint_capsule
   center=<4.3018 0 -0.0608>
   rotation=<rot -87.9531 179.9983 -179.9992>
  }]
 dyn_feather_6_joint [{
   shapeType=Capsule
   length=4.5
   radius=1.5
   shapeName=dyn_feather_6_joint_capsule
   center=<5.1079 0 -0.2136>
   rotation=<rot -86.8803 179.9982 -179.9987>
  }]
 dyn_feather_7_joint [{shapeType=Capsule, length=2.5, radius=1, shapeName=dyn_feather_7_joint_capsule, center=<3.3955 0 -0.2971>, rotation=<rot -85 180 179.9997>}]
 dyn_feather_8_joint [{
   shapeType=Capsule
   length=2
   radius=1
   shapeName=dyn_feather_8_joint_capsule
   center=<3.0825 0 -0.2697>
   rotation=<rot -85 -179.9995 179.9993>
  }]
 dyn_feather_9_joint [{
   shapeType=Capsule
   length=3.5
   radius=1
   shapeName=dyn_feather_9_joint_capsule
   center=<4.4493 0 -0.3893>
   rotation=<rot -85.0001 179.9988 -179.9994>
  }]
}
constraints [
 {bone1Name=dyn_feather_1_joint, bone2Name=dyn_base_joint, swing1LimitDegrees=5, swing2LimitDegrees=30, twistLimitDegrees=0}
 {bone1Name=dyn_feather_2_joint, bone2Name=dyn_feather_1_joint, swing1LimitDegrees=5, swing2LimitDegrees=50, twistLimitDegrees=0}
 {bone1Name=dyn_feather_3_joint, bone2Name=dyn_base_joint, swing1LimitDegrees=5, swing2LimitDegrees=30, twistLimitDegrees=0}
 {bone1Name=dyn_feather_5_joint, bone2Name=dyn_base_joint, swing1LimitDegrees=5, swing2LimitDegrees=10, twistLimitDegrees=0}
 {bone1Name=dyn_feather_7_joint, bone2Name=dyn_base_joint, swing1LimitDegrees=40, swing2LimitDegrees=8, twistLimitDegrees=0}
 {bone1Name=dyn_feather_9_joint, bone2Name=dyn_base_joint, swing1LimitDegrees=10, swing2LimitDegrees=5, twistLimitDegrees=0}
 {bone1Name=dyn_feather_11_joint, bone2Name=dyn_base_joint, swing1LimitDegrees=30, swing2LimitDegrees=5, twistLimitDegrees=0}
 {bone1Name=dyn_feather_13_joint, bone2Name=dyn_base_joint, swing1LimitDegrees=5, swing2LimitDegrees=30, twistLimitDegrees=0}
 {bone1Name=dyn_feather_15_joint, bone2Name=dyn_base_joint, swing1LimitDegrees=3, swing2LimitDegrees=30, twistLimitDegrees=0}
 {bone1Name=dyn_feather_4_joint, bone2Name=dyn_feather_3_joint, swing1LimitDegrees=5, swing2LimitDegrees=40, twistLimitDegrees=0}
 {bone1Name=dyn_feather_6_joint, bone2Name=dyn_feather_5_joint, swing1LimitDegrees=5, swing2LimitDegrees=40, twistLimitDegrees=0}
 {bone1Name=dyn_feather_8_joint, bone2Name=dyn_feather_7_joint, swing1LimitDegrees=40, swing2LimitDegrees=10, twistLimitDegrees=0}
 {bone1Name=dyn_feather_10_joint, bone2Name=dyn_feather_9_joint, swing1LimitDegrees=40, swing2LimitDegrees=5, twistLimitDegrees=0}
 {bone1Name=dyn_feather_12_joint, bone2Name=dyn_feather_11_joint, swing1LimitDegrees=40, swing2LimitDegrees=10, twistLimitDegrees=0}
 {bone1Name=dyn_feather_14_joint, bone2Name=dyn_feather_13_joint, swing1LimitDegrees=5, swing2LimitDegrees=40, twistLimitDegrees=0}
 {bone1Name=dyn_feather_16_joint, bone2Name=dyn_feather_15_joint, swing1LimitDegrees=5, swing2LimitDegrees=30, twistLimitDegrees=0}
 {
  bone1Name=dyn_claw_1_joint
  bone2Name=dyn_base_joint
  swing1LimitDegrees=0
  swing1Motion=Locked
  swing2LimitDegrees=15
  twistLimitDegrees=0
  twistMotion=Locked
 }
 {
  bone1Name=dyn_claw_2_joint
  bone2Name=dyn_base_joint
  swing1LimitDegrees=0
  swing1Motion=Locked
  swing2LimitDegrees=15
  twistLimitDegrees=0
  twistMotion=Locked
 }
 {
  bone1Name=dyn_claw_3_joint
  bone2Name=dyn_base_joint
  swing1LimitDegrees=0
  swing1Motion=Locked
  swing2LimitDegrees=15
  twistLimitDegrees=0
  twistMotion=Locked
 }
]
